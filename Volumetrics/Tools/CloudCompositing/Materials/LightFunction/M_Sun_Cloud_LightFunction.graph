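# Light function material for the sun that projects cloud shadows from a 2D cloud texture onto the world.
# World position is traced along the MPC_Clouds light direction to a cloud altitude plane in kilometers, forming texture UVs with collection-driven scale, tiling and offset.
# The mip-selected, invertible sample feeds two branches: Beer's-law exp(-x) shadowing with bias, density and contrast, and a raw value fading to unshadowed near cloud altitude.
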Material /Volumetrics/Tools/CloudCompositing/Materials/LightFunction/M_Sun_Cloud_LightFunction
graph {
 properties {materialDomain=MD_LightFunction, pixelDepthOffsetMode=PDOM_Legacy, usedShadingModels=Default Lit}
 nodes [
  n1 = WorldPosition()
  n2 = WorldPosition()
  n3 = Subtract(
   A=ScalarParameter() {defaultValue=5, parameterName=Cloud Altitude KM}
   B=MaterialFunctionCall!(Cm=ComponentMask(None=@n2.XYZ) {b=true}) {materialFunction=/Engine/Functions/Engine_MaterialFunctions01/Units/Cm-to-Km}.Km
  )
  n4 = Abs(None=ComponentMask(None=CollectionParameter() {collection=/Volumetrics/Tools/CloudCompositing/MPC/MPC_Clouds, parameterName=LightDirection}) {b=true})
  n5 = CollectionParameter() {collection=&MPC_Clouds, parameterName=LightDirection}
  n6 = Subtract(
   A=ComponentMask(None=Divide(A=@n1.XYZ) {constB=100000, desc=cm to km}) {g=true, r=true}
   B=Multiply(A=Divide(A=@n3, B=@n4), B=ComponentMask(None=@n5) {g=true, r=true})
  )
  n7 = Divide(
   A=CollectionParameter() {collection=&MPC_Clouds, parameterName=SkyTextureScaleKm}
   B=ScalarParameter() {defaultValue=1, parameterName=Repeat Tiling}
  )
  n8 = VectorParameter() {parameterName=Offset, defaultValue=<rgba 0.5 0.5 0 0>}
  n9 = TextureSampleParameter2D!(
   UVs=Subtract(A=Divide(A=@n6, B=@n7), B=ComponentMask(None=@n8.RGB) {g=true, r=true})
   Level=ScalarParameter() {parameterName=Mip Level}
   Apply View MipBias
  ) {mipValueMode=TMVM_MipLevel, parameterName=Texture, texture=/Engine/EditorResources/BadBlueprintSprite}
  n10 = LinearInterpolate(A=@n9.R, B=OneMinus(None=@n9.R), Alpha=ScalarParameter() {parameterName=Invert})
  n11 = OneMinus(None=Saturate(None=@n10))
  n12 = ScalarParameter() {parameterName=Bias}
  MaterialExpressionMultiply_0 = Multiply(A=Max(A=Subtract(A=@n11, B=@n12)) {constB=0}, B=ScalarParameter() {defaultValue=2, parameterName=Density})
  n13 = Custom!(x=@MaterialExpressionMultiply_0) {
   code="return exp(-x);"
   description=Exp
   outputType=CMOT_Float1
   inputs [{
     inputName=x
     input {expression=&self:MaterialExpressionMultiply_0}
    }]
  }
  LinearInterpolate(B=@n13, Alpha=ScalarParameter() {defaultValue=1, parameterName=Contrast}) {constA=1}
  LinearInterpolate(B=@n10, Alpha=Saturate(None=Divide(A=@n3) {constB=0.05})) {constA=1}
 ]
}
